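# Gemstone surface material using Subsurface shading that fakes refraction and internal reflections, usable on skeletal meshes.
# A distance-corrected cubemap function traces refracted rays into a cubemap, whose colors decode as interior normals for a second Refract step.
# Sky light environment samples, sun highlights (custom GGX, cheaper power lobe on mobile) and Schlick-phase scattering provide lighting terms.
# Tint, depth gradient and subsurface terms shape the look, with emissive divided by eye adaptation.
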Material /Game/Global/Materials/MaterialGym/Materials/Gem/M_GemStone
graph {
 properties {
  bUseMaterialAttributes=true
  bUsedWithMeshDeformer=true
  bUsedWithSkeletalMesh=true
  pixelDepthOffsetMode=PDOM_Legacy
  shadingModel=MSM_Subsurface
  usedShadingModels=Subsurface
 }
 nodes [
  Constant3Vector()
  LinearInterpolate(A=NamedRerouteUsage(), B=ScalarParameter() {defaultValue=0.5, parameterName=B}, Alpha=NamedRerouteUsage()) {constA=1, constB=0}
  n1 = MaterialFunctionCall!(
   CubePosition=Add(A=ActorPositionWS(), B=VectorParameter() {group=Gemstone, parameterName=GS_CubeMapPositionOffset, defaultValue=<rgba 0 0 0 1>}.RGB)
   Radius=ScalarParameter() {defaultValue=150, group=Gemstone, parameterName=GS_RefractedCubeMapRadius}
   UseRefractedRay=StaticBool() {value=true}
   IOR=ScalarParameter() {defaultValue=1, parameterName=GS_IncomingIOR}
   CustomNormalWS=Transform(None=TextureSampleParameter2D() {group=Gemstone, parameterName=GS_NormalMap, samplerType=SAMPLERTYPE_Normal, texture=/Engine/EngineMaterials/FlatNormal}.RGB)
  ) {materialFunction=/Game/Global/Materials/MaterialGym/Materials/Gem/MaterialFunctions/MF_DistanceCorrectedCubemapSample}
  n2 = TextureSampleParameterCube!(UVs=NamedRerouteDeclaration(None=@n1.DistanceCorrectedRay) {name=UVW, nodeColor=<rgba 0 0.625 1 1>}, Apply View MipBias) {
   group=Gemstone
   parameterName=GS_RefractionCubeMap
   samplerType=SAMPLERTYPE_LinearColor
   texture=/Game/Global/Materials/MaterialGym/Materials/Gem/Textures/TC_InteriorRoomCube
  }
  Multiply(A=@n2.RGB, B=NamedRerouteDeclaration(None=@n1.Mask) {name=CubeMask, nodeColor=<rgba 0 1 0.6016 1>})
  n3 = Multiply(A=SkyLightEnvMapSample(Direction=NamedRerouteUsage(), Roughness=ScalarParameter() {defaultValue=0.2, group=Gemstone, parameterName=GS_ReflectRoughness})) {constB=0.75}
  n4 = SkyLightEnvMapSample(Direction=NamedRerouteUsage(), Roughness=ScalarParameter() {defaultValue=0.2, group=Gemstone, parameterName=GS_RefractRoughness})
  n5 = Multiply(A=Multiply(A=@n4) {constB=1.2}, B=NamedRerouteUsage())
  n6 = NamedRerouteUsage()
  n7 = Saturate(None=Multiply(A=LinearInterpolate(A=@n3, B=@n5, Alpha=@n6), B=ScalarParameter() {defaultValue=1, group=Gemstone, parameterName=GS_ReflectionsBrightness}))
  n8 = ScalarParameter() {defaultValue=1, group=Gemstone, parameterName=GS_RefractionsExponent}
  n9 = Multiply(
   A=NamedRerouteUsage()
   B=Desaturation(None=Power(Base=@n7, Exp=@n8), Fraction=ScalarParameter() {group=Gemstone, parameterName=GS_RefractionsDesaturation})
  )
  Multiply(A=NamedRerouteUsage(), B=@n9) {constA=1}
  NamedRerouteDeclaration(None=ScalarParameter() {defaultValue=1, group=Gemstone, parameterName=GS_IOR}) {name=IOR, nodeColor=<rgba 1 0 0.2109 1>}
  n10 = NamedRerouteUsage()
  n11 = ConstantBiasScale(None=@n2.RGB) {bias=-0.5, scale=2}
  n12 = NamedRerouteDeclaration(None=Multiply(A=Normalize(VectorInput=@n11)) {constB=-1}) {name=InteriorNormal, nodeColor=<rgba 0.5234 1 0 1>}
  n13 = NamedRerouteUsage()
  n14 = Constant() {r=1}
  NamedRerouteDeclaration(None=Multiply(A=MaterialFunctionCall!(Ray Direction=@n10, Surface Normal=@n12, Refractive Index Origin=@n13, Refractive Index Target=@n14) {materialFunction=/Engine/Functions/Engine_MaterialFunctions01/Vectors/Refract}.Vector) {constB=-1}) {name=RefractVectorInterior, nodeColor=<rgba 1 0.9375 0 1>}
  MaterialExpressionNormalize_2 = Normalize(VectorInput=SkyAtmosphereLightDirection())
  MaterialExpressionCameraVectorWS_1 = CameraVectorWS()
  MaterialExpressionNamedRerouteUsage_8 = NamedRerouteUsage()
  MaterialExpressionConstant_28 = Constant() {r=0.3}
  MaterialExpressionConstant3Vector_3 = Constant3Vector() {constant=<rgba 1 1 1 0>}
  n15 = Custom!(
   L=@MaterialExpressionNormalize_2
   V=@MaterialExpressionCameraVectorWS_1
   N=@MaterialExpressionNamedRerouteUsage_8
   roughness=@MaterialExpressionConstant_28
   F0=@MaterialExpressionConstant3Vector_3
  ) {
   code="    float3 H = normalize(L + V);\r\n    float NdotH = saturate(dot(N, H));\r\n    float NdotL = saturate(dot(N, L));\r\n    float NdotV = saturate(dot(N, V));\r\n    float3 F = F0 + (1.0 - F0) * pow(1.0 - NdotV, 5.0);\r\n    float epsilon = .0001;\r\n\r\n\tfloat alpha = roughness * roughness;\r\n    float alphaSqr = alpha * alpha;\r\n\r\n    float denominator = NdotH * NdotH * (alphaSqr - 1.0) + 1.0;\r\n    float D = alphaSqr / (PI * denominator * denominator + epsilon);\r\n\t\r\n\t\r\n\t\r\n    float k = alpha / 2;    \r\n    float g_v = NdotV / ( NdotV*(1 - k ) + k);\r\n    float g_l = NdotL / (NdotL*(1 - k ) + k);\r\n    float G = g_v * g_l;\r\n    return NdotL*(D * G * F) / (4.0 * NdotL * NdotV + epsilon );"
   description=GGX
   inputs [
    {inputName=L, input {expression=&self:MaterialExpressionNormalize_2}}
    {inputName=V, input {expression=&self:MaterialExpressionCameraVectorWS_1}}
    {inputName=N, input {expression=&self:MaterialExpressionNamedRerouteUsage_8}}
    {inputName=roughness, input {expression=&self:MaterialExpressionConstant_28}}
    {
     inputName=F0
     input {
      expression=&self:MaterialExpressionConstant3Vector_3
      mask=1
      maskB=1
      maskG=1
      maskR=1
     }
    }
   ]
  }
  n16 = Multiply(A=@n15) {constA=1}
  n17 = DotProduct(A=NamedRerouteUsage(), B=@MaterialExpressionNormalize_2)
  n18 = Multiply(A=Power(Base=Saturate(None=@n17)) {constExponent=24}) {constB=5}
  n19 = NamedRerouteUsage()
  n20 = Multiply(A=Clamp(None=ShadingPathSwitch(Default=@n16, Mobile=@n18)) {maxDefault=40}, B=Multiply(A=@n19, B=@n19))
  n21 = NamedRerouteUsage()
  n22 = NamedRerouteUsage()
  NamedRerouteDeclaration(None=Multiply(A=Multiply(A=@n20, B=@n21), B=LinearInterpolate(Alpha=@n22) {constA=0.5, constB=0.05})) {name=SunInternalReflections, nodeColor=<rgba 0 0.8828 1 1>}
  NamedRerouteDeclaration(None=@n9) {name=Sky Refractions, nodeColor=<rgba 0.6875 1 0 1>}
  NamedRerouteDeclaration(None=LinearInterpolate(
    A=Constant3Vector() {constant=<rgba 0 0 1 0>}
    B=Multiply(A=@n1.DistanceCorrectedRay, B=@n1.DistanceCorrectedRay)
    Alpha=ScalarParameter() {parameterName=Param_2}
   ) {constAlpha=0}) {name=DebugNormal, nodeColor=<rgba 0.6875 0 1 1>}
  n23 = Constant() {r=1}
  n24 = ComponentMask(None=NamedRerouteUsage()) {b=true}
  n25 = Saturate(None=OneMinus(None=Multiply(A=@n24) {constB=-1}))
  NamedRerouteDeclaration(None=MaterialFunctionCall!(1=@n23, 2=Constant() {r=0.5}, 3=@n23, 4, A=Multiply(A=@n25, B=@n25)) {materialFunction=/Engine/Functions/Engine_MaterialFunctions02/Lerp_Multiple_Float}.Lerp 3 Inputs) {name=GroundGradient, nodeColor=<rgba 0.4062 0 1 1>}
  n26 = Multiply(A=Multiply(A=NamedRerouteUsage()) {constB=0.5}) {constB=0.5}
  n27 = NamedRerouteUsage()
  n28 = SkyAtmosphereLightDirection()
  n29 = LinearInterpolate(A=ScalarParameter() {defaultValue=0.5, parameterName=A}, Alpha=Abs(None=DotProduct(A=@n27, B=@n28))) {constA=0.25}
  NamedRerouteDeclaration(None=Multiply(A=Multiply(A=@n26, B=@n29), B=ScalarParameter() {defaultValue=1, group=Gemstone, parameterName=GS_SS_Strength})) {name=SubSurface, nodeColor=<rgba 1 0 0.5859 1>}
  n30 = NamedRerouteUsage()
  n31 = NamedRerouteUsage()
  NamedRerouteDeclaration(None=Add(A=Add(A=@n30, B=@n31), B=NamedRerouteUsage())) {name=Emissive, nodeColor=<rgba 1 0.9609 0 1>}
  NamedRerouteDeclaration(None=@n4) {name=debug, nodeColor=<rgba 0.1719 0 1 1>}
  n32 = Multiply(A=DotProduct(A=NamedRerouteUsage(), B=NamedRerouteUsage())) {constB=-1}
  n33 = ScalarParameter() {defaultValue=0.73}
  n34 = ScalarParameter() {defaultValue=0.78, parameterName=Param_1}
  NamedRerouteDeclaration(None=LinearInterpolate(
    A=ScalarParameter() {defaultValue=0.5, group=Gemstone, parameterName=GS_ReflectionAngleMin}
    Alpha=MaterialFunctionCall(Input=@n32, Input Low=@n33, Input High=@n34) {materialFunction=&RemapValue_0-1}.Result
   ) {constA=0.5}) {name=TIR, nodeColor=<rgba 0 0.7422 1 1>}
  NamedRerouteDeclaration(None=@n1.RefractedRay) {name=RefractedRay, nodeColor=<rgba 1 0 0.8906 1>}
  n35 = NamedRerouteUsage()
  n36 = Constant() {r=-1}
  n37 = Constant() {r=1}
  NamedRerouteDeclaration(None=LinearInterpolate(
    A=ScalarParameter() {defaultValue=0.5, group=Gemstone, parameterName=GS_GradientLow}
    B=ScalarParameter() {defaultValue=1, group=Gemstone, parameterName=GS_GradientHigh}
    Alpha=MaterialFunctionCall(Input=@n35, Input Low=@n36, Input High=@n37) {materialFunction=&RemapValue_0-1}.Result
   ) {constA=0.45}) {name=Depth Gradient, nodeColor=<rgba 1 0 0.2812 1>}
  n38 = VectorParameter() {group=Gemstone, parameterName=InColor, defaultValue=<rgba 0 0 0 1>}
  n39 = VectorParameter() {group=Gemstone, parameterName=GS_ColorTint, sortPriority=-2, defaultValue=<rgba 0.0272 0.5312 0.5101 1>}
  n40 = ScalarParameter() {defaultValue=1, group=Gemstone, parameterName=GS_BaseColorBlend, sortPriority=-1}
  NamedRerouteDeclaration(None=Desaturation(None=LinearInterpolate(A=@n38.RGB, B=@n39.RGB, Alpha=@n40), Fraction=Constant() {r=-0.2})) {name=Tint, nodeColor=<rgba 0.3125 1 0 1>}
  NamedRerouteDeclaration(None=ReflectionVectorWS(CustomWorldNormal=NamedRerouteUsage())) {name=ReflectionVector, nodeColor=<rgba 1 0 0.3281 1>}
  n41 = Constant3Vector() {constant=<rgba 0 0 100 0>}
  NamedRerouteDeclaration(None=Multiply(A=SkyAtmosphereLightIlluminance!(World Position=@n41)) {constB=0.5}) {name=DirectionalLightColor, nodeColor=<rgba 0 0.4844 1 1>}
  n42 = DotProduct(A=SkyAtmosphereLightDirection(), B=NamedRerouteUsage())
  n43 = ScalarParameter() {defaultValue=-0.985, group=Gemstone, parameterName=GS_Phase}
  n44 = Clamp(None=Multiply(
    A=NamedRerouteUsage()
    B=MaterialFunctionCall!(Cosine=@n42, Anisotropy=@n43) {materialFunction=/Engine/Functions/Engine_MaterialFunctions01/Volumetrics/SchlickPhase}.Result
   )) {maxDefault=20}
  n45 = NamedRerouteUsage()
  n46 = NamedRerouteUsage()
  NamedRerouteDeclaration(None=Multiply(A=Multiply(A=@n44, B=@n45) {constB=2}, B=LinearInterpolate(Alpha=@n46) {constA=0.2})) {name=SunNscatter, nodeColor=<rgba 1 0.7031 0 1>}
  NamedRerouteUsage()
  NamedRerouteUsage()
  ScalarParameter() {defaultValue=1.5, parameterName=TestIOR}
  n47 = ShadingPathSwitch(Default=Constant(), Mobile=NamedRerouteUsage())
  n48 = NamedRerouteUsage()
  SetMaterialAttributes!(
   MaterialAttributes
   Normal=Constant3Vector() {constant=<rgba 0 0 1 0>}
   Metallic=ScalarParameter() {group=Gemstone, parameterName=GS_Metallic}
   Roughness=ScalarParameter() {defaultValue=0.1, group=Gemstone, parameterName=GS_Roughness}
   Emissive Color=Divide(A=Add(A=@n47, B=@n48) {constB=0}, B=EyeAdaptation().EyeAdaptation)
   Opacity=ScalarParameter() {defaultValue=0.7, group=Gemstone, parameterName=GS_Opacity}
   Subsurface Color=Multiply(A=NamedRerouteUsage(), B=ScalarParameter() {parameterName=GS_SubsurfaceStrength})
   Specular=ScalarParameter() {defaultValue=1, group=Gemstone, parameterName=GS_Specular}
   Base Color=VectorParameter() {parameterName=GS_BaseColor, defaultValue=<rgba 0 0 0 1>}.RGB
  ) {attributeSetTypes [
    #1
    #2
    #3
    #4
    #5
    #6
    #7
    #8
   ]}
  Step(X=Saturate(None=@n1.Mask)) {constX=0.5, constY=0.5}
 ]
}
